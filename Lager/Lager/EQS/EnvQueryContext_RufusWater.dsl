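; EQS query context that supplies the level's water actors to environment queries, named for Rufus.
; ProvideActorsSet gathers every FortAthenaWater and WaterVolume_Athena actor through two GetAllActorsOfClass searches and appends both sets to one returned array.
; The querier inputs are unused, so the result is not filtered by distance or relation to the querier.
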(fn ProvideActorsSet
  (params (QuerierObject :type Object%20Reference) (QuerierActor :type Actor%20Object%20Reference))
  (bind _outactors (Actor|GetAllActorsOfClass "/Script/FortniteGame.FortAthenaWater"))
  (for _array_element _outactors
    (Utilities|Array|Add (LocalVariables|GetReturnValue) _array_element))
  (bind _outactors_2 (Actor|GetAllActorsOfClass "/Script/FortniteGame.WaterVolume_Athena"))
  (for _array_element_2 _outactors_2
    (Utilities|Array|Add (LocalVariables|GetReturnValue) _array_element_2))
  (return (LocalVariables|GetReturnValue)))
