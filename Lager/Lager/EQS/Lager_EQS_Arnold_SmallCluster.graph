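# An environment query, apparently for spawning or placing an Arnold small cluster in Lager, that finds open spots free of nearby buildings.
# It generates a simple grid of points projected onto navmesh with a filter excluding all water, lifted slightly above the surface.
# A nearby-buildings test then filters and scores each point, requiring no buildings close by while ignoring floor pieces.
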EnvQuery /Lager/Lager/EQS/Lager_EQS_Arnold_SmallCluster
options [{
  generator {
   name=EnvQueryGenerator_SimpleGrid_6
   class=EnvQueryGenerator_SimpleGrid
   properties {
    gridSize=800
    spaceBetween=400
    projectionData {
     extentX=0
     navigationFilter=&NavFilter_NPC_ExcludeAllWater_C
     postProjectionVerticalOffset=25
     projectDown=600
     projectUp=600
    }
   }
  }
  tests [FortQueryTest_HasNearbyBuildings_7 {
    class=FortQueryTest_HasNearbyBuildings
    properties {bIncludeFloors=false, extentXY=2, testComment="Require no building nearby (Floor allowed)", testPurpose=FilterAndScore}
   }]
 }]
