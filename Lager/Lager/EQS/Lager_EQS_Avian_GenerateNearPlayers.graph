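# EQS query for Lager avian AI that finds landing or perch points on buildings around the querier, presumably a player given the asset name.
# A Fort generator samples a few points on props, roofs, walls and stairs, excluding tire containers and landing-prohibited actors.
# Filter-only tests box-trace points onto geometry, reject those near Living World wards, and enforce a minimum 2D distance from the querier.
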EnvQuery /Lager/Lager/EQS/Lager_EQS_Avian_GenerateNearPlayers
options [{
  generator {
   name=FortQueryGenerator_PointsOnBuildingActors_0
   class=FortQueryGenerator_PointsOnBuildingActors
   properties {
    bAutoSortTests=true
    bCanRunAsync=false
    boundingBoxExtentXY=4000
    boundingBoxExtentZ=2000
    envQueryResultNormalizationOption=Default
    generateAround=&EnvQueryContext_Querier
    maxGeneratedPoints=10
    navDataOverrideContext=null
    optionName=""
    pointDensity=1000
    randomChanceToSkip=0
    actorLookupFilter {
     minDistanceFromBoundsCenter=0
     bounds {sphereRadius=0, boxExtent=<0 0 0>, origin=<0 0 0>}
     optionalSubclasses [&BuildingProp, &BuildingRoof, &BuildingWall, &BuildingStairs]
     requiredProperties []
     tagQuery {
      autoDescription=" NONE( Building.Type.Container.Tires, Building.AI.Prohibits.Landing )"
      tokenStreamVersion=0
      userDescription=""
      queryTokenStream [0,1,3,2,0,1]
      tagDictionary [{tagName=Building.Type.Container.Tires}, {tagName=Building.AI.Prohibits.Landing}]
     }
    }
    projectionData {
     bCanDisableTrace=true
     bCanProjectDown=true
     bCanTraceOnGeometry=true
     bCanTraceOnNavMesh=true
     bOnlyBlockingHits=true
     bTraceComplex=false
     extentX=0
     extentY=10
     extentZ=10
     navigationFilter=null
     postProjectionVerticalOffset=0
     projectDown=1024
     projectUp=1024
     serializedChannel=ECC_Visibility
     traceChannel=TraceTypeQuery1
     traceMode=None
     traceProfileName=None
     traceShape=Line
     versionNum=1
    }
   }
  }
  tests [
   EnvQueryTest_Project_12 {
    class=EnvQueryTest_Project
    properties {
     scoringFactor=1
     testComment=""
     testPurpose=Filter
     projectionData {
      extentY=20
      extentZ=20
      navigationFilter=null
      postProjectionVerticalOffset=0
      projectDown=1024
      projectUp=4000
      serializedChannel=ECC_GameTraceChannel11
      traceChannel=TraceTypeQuery10
      traceMode=GeometryByChannel
      traceShape=Box
     }
    }
   }
   FortQueryTest_IsCloseToLivingWorldWard_0 {
    class=FortQueryTest_IsCloseToLivingWorldWard
    properties {
     bDefineReferenceValue=false
     boolValue=false
     clampMaxType=None
     clampMinType=None
     filterType=Match
     floatValueMax=0
     floatValueMin=0
     multipleContextFilterOp=AllPass
     multipleContextScoreOp=AverageScore
     normalizationType=Absolute
     referenceValue=0
     scoreClampMax=0
     scoreClampMin=0
     scoringEquation=Constant
     scoringFactor=1
     testComment=""
     testPurpose=Filter
    }
   }
   EnvQueryTest_Distance_0 {
    class=EnvQueryTest_Distance
    properties {
     distanceTo=&EnvQueryContext_Querier
     filterType=Minimum
     floatValueMin=500
     scoringEquation=Linear
     scoringFactor=1
     testComment=""
     testMode=Distance2D
     testPurpose=Filter
    }
   }
  ]
 }]
