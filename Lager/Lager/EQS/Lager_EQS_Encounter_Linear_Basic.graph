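# EQS query that finds encounter spawn points on a navmesh-projected grid around the querier, apparently a spawner with VFX.
# Three options share the same grid and relax in turn as fallbacks.
# The first filters for a clear upward box sweep, minimum 2D distance from the VFX center, and a clear visual trace to the spawner.
# The second drops the trace, and the last accepts any projected point untested.
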EnvQuery /Lager/Lager/EQS/Lager_EQS_Encounter_Linear_Basic
options [
 {
  generator {
   name=EnvQueryGenerator_SimpleGrid_0
   class=EnvQueryGenerator_SimpleGrid
   properties {
    gridSize {dataBinding=null, dataField=FloatValue, defaultValue=600}
    projectionData {
     extentX=100
     navigationFilter=&RecastFilter_UseDefaultArea
     postProjectionVerticalOffset=30
     projectDown=200
     projectUp=200
    }
    spaceBetween {dataBinding=null, dataField=FloatValue, defaultValue=200}
   }
  }
  tests [
   EnvQueryTest_Project_4 {
    class=EnvQueryTest_Project
    properties {
     boolValue=false
     scoringFactor=-1
     testComment=Require clear sweep up
     testPurpose=Filter
     projectionData {
      extentY=100
      extentZ=100
      navigationFilter=&RecastFilter_UseDefaultArea
      postProjectionVerticalOffset=0
      projectDown=0
      projectUp=200
      traceMode=GeometryByChannel
      traceShape=Box
     }
    }
   }
   EnvQueryTest_Distance_0 {
    class=EnvQueryTest_Distance
    properties {
     distanceTo=&EnvQueryContext_Querier
     filterType=Minimum
     floatValueMin=300
     scoringEquation=Linear
     scoringFactor=1
     testComment=Don't spawn right on the center of our VFX
     testMode=Distance2D
     testPurpose=Filter
    }
   }
   EnvQueryTest_Trace_0 {
    class=EnvQueryTest_Trace
    properties {contextHeightOffset=50, scoringFactor=1, testComment="Look for clear visual path to spawner, if it exists.", testPurpose=Filter}
   }
  ]
 }
 {
  generator {
   name=EnvQueryGenerator_SimpleGrid_4
   class=EnvQueryGenerator_SimpleGrid
   properties {
    gridSize {dataBinding=null, dataField=FloatValue, defaultValue=600}
    projectionData {
     extentX=100
     navigationFilter=&RecastFilter_UseDefaultArea
     postProjectionVerticalOffset=30
     projectDown=200
     projectUp=200
    }
    spaceBetween {dataBinding=null, dataField=FloatValue, defaultValue=200}
   }
  }
  tests [
   EnvQueryTest_Project_1 {
    class=EnvQueryTest_Project
    properties {
     boolValue=false
     scoringFactor=-1
     testComment=Require clear sweep up
     testPurpose=Filter
     projectionData {
      extentY=100
      extentZ=100
      navigationFilter=&RecastFilter_UseDefaultArea
      postProjectionVerticalOffset=0
      projectDown=0
      projectUp=200
      traceMode=GeometryByChannel
      traceShape=Box
     }
    }
   }
   EnvQueryTest_Distance_2 {
    class=EnvQueryTest_Distance
    properties {
     distanceTo=&EnvQueryContext_Querier
     filterType=Minimum
     floatValueMin=300
     scoringEquation=Linear
     scoringFactor=1
     testComment=Don't spawn right on the center of our VFX
     testMode=Distance2D
     testPurpose=Filter
    }
   }
  ]
 }
 {
  generator {
   name=EnvQueryGenerator_SimpleGrid_1
   class=EnvQueryGenerator_SimpleGrid
   properties {
    gridSize {dataBinding=null, dataField=FloatValue, defaultValue=600}
    projectionData {
     extentX=100
     navigationFilter=&RecastFilter_UseDefaultArea
     postProjectionVerticalOffset=30
     projectDown=200
     projectUp=200
    }
    spaceBetween {dataBinding=null, dataField=FloatValue, defaultValue=200}
   }
  }
  tests []
 }
]
