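# Environment query that finds floating spawn points above the ground, outside water, for a forest variant per the asset name.
# A small grid around the default context projects onto navmesh with a water-excluding NPC filter, then lifts each point off the surface.
# An inverted geometry projection test filters for clear space overhead, and a Fortnite nearby-buildings test filters and scores to avoid structures.
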EnvQuery /Lager/Lager/EQS/Lager_EQS_FloatingPointAroundSpawn_ForestExcludeWater
options [{
  generator {
   name=EnvQueryGenerator_SimpleGrid_1
   class=EnvQueryGenerator_SimpleGrid
   properties {
    bAutoSortTests=false
    gridSize {dataBinding=null, dataField=FloatValue, defaultValue=300}
    projectionData {
     extentX=0
     navigationFilter=&NavFilter_NPC_ExcludeAllWater_C
     postProjectionVerticalOffset=100
     projectDown=3000
     projectUp=3000
    }
    spaceBetween {dataBinding=null, dataField=FloatValue, defaultValue=200}
   }
  }
  tests [
   EnvQueryTest_Project_0 {
    class=EnvQueryTest_Project
    properties {
     boolValue=false
     scoringFactor=-1
     testComment=Require clear sweep up
     testPurpose=Filter
     projectionData {
      extentY=0
      extentZ=100
      navigationFilter=&RecastFilter_UseDefaultArea
      postProjectionVerticalOffset=0
      projectDown=0
      projectUp=800
      traceMode=GeometryByChannel
     }
    }
   }
   FortQueryTest_HasNearbyBuildings_0 {class=FortQueryTest_HasNearbyBuildings, properties {extentXY=2, testComment=Require no building nearby, testPurpose=FilterAndScore}}
  ]
 }]
