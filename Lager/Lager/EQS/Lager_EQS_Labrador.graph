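# Environment query for Labrador in Lager that picks candidate points away from water, generated in the querier's bounding volume.
# A Fortnite generator scatters random points within the volume without a random Z and projects them onto geometry with a vertical line trace.
# A navigation projection test uses a water-only nav filter with an inverted match, so points in or on water are removed.
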EnvQuery /Lager/Lager/EQS/Lager_EQS_Labrador
options [{
  generator {
   name=FortQueryGenerator_RandomPointsInBoundingVolume_1
   class=FortQueryGenerator_RandomPointsInBoundingVolume
   properties {
    bAddRandomZValueFromBoundingVolume=false
    bAutoSortTests=true
    bCanRunAsync=false
    envQueryResultNormalizationOption=Default
    generateIn=&EnvQueryContext_Querier
    navDataOverrideContext=null
    optionName=""
    randomPointsCount=100
    extendBoundingVolumeBounds=<0 0 0>
    projectionData {
     bCanDisableTrace=true
     bCanProjectDown=true
     bCanTraceOnGeometry=true
     bCanTraceOnNavMesh=true
     bOnlyBlockingHits=true
     bTraceComplex=false
     extentX=0
     extentY=10
     extentZ=10
     navigationFilter=null
     postProjectionVerticalOffset=0
     projectDown=5000
     projectUp=10000
     serializedChannel=ECC_Visibility
     traceChannel=TraceTypeQuery1
     traceMode=GeometryByChannel
     traceProfileName=None
     traceShape=Line
     versionNum=1
    }
   }
  }
  tests [EnvQueryTest_Project_0 {
    class=EnvQueryTest_Project
    properties {
     boolValue=false
     scoringFactor=1
     testComment=Remove all points in/on water
     testPurpose=FilterAndScore
     projectionData {
      navigationFilter=/Game/Athena/AI/NavFilters/NavFilter_NPC_WaterOnly.NavFilter_NPC_WaterOnly_C
      postProjectionVerticalOffset=0
      projectDown=5000
      projectUp=10000
      traceMode=Navigation
     }
    }
   }]
 }]
