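# EQS query that finds points inside a FortAthenaLivingWorldVolume, supplied through the Querier context, for the Lager living-world system.
# A SimpleGrid generator projects points onto default-area navmesh, with size bound to a query parameter and spacing to a game-difficulty float curve.
# A Volume test filters and scores points against the volume.
# A Project test box-traces upward, keeping only points with a clear sweep above them.
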EnvQuery /Lager/Lager/EQS/Lager_EQS_PointInVolume_GenerateToVolumeBounds_DefaultNav
options [{
  generator {
   name=EnvQueryGenerator_SimpleGrid_0
   class=EnvQueryGenerator_SimpleGrid
   properties {
    gridSize {
     dataBinding=&self:EnvQueryGenerator_SimpleGrid_0.AIDataProvider_QueryParams_0
     dataField=FloatValue
     defaultValue=20000
    }
    projectionData {
     extentX=0
     navigationFilter=&RecastFilter_UseDefaultArea
     postProjectionVerticalOffset=0
     projectDown=3000
     projectUp=3000
    }
    spaceBetween {
     dataBinding=&self:EnvQueryGenerator_SimpleGrid_0.FortAIDataProvider_FloatCurveOverGameDifficulty_0
     dataField=FloatValue
     defaultValue=2500
    }
   }
  }
  tests [
   EnvQueryTest_Volume_0 {
    class=EnvQueryTest_Volume
    properties {
     boolValue=true
     testPurpose=FilterAndScore
     volumeClass=/Script/LagerRuntime.FortAthenaLivingWorldVolume
     volumeContext=&EnvQueryContext_Querier
    }
   }
   EnvQueryTest_Project_0 {
    class=EnvQueryTest_Project
    properties {
     boolValue=false
     scoringFactor=-1
     testComment=Require clear sweep up
     testPurpose=Filter
     projectionData {
      extentY=100
      extentZ=100
      navigationFilter=&RecastFilter_UseDefaultArea
      postProjectionVerticalOffset=0
      projectDown=0
      projectUp=3000
      traceMode=GeometryByChannel
      traceShape=Box
     }
    }
   }
  ]
 }]
