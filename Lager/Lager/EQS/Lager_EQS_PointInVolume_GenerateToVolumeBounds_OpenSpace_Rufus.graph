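# Lager EQS query that finds open-space points, clear of Rufus water and indestructible buildings, within the querier's living-world volume.
# A simple grid, sized by query param and spaced by a game-difficulty curve, projects with a nav filter excluding all water.
# Tests filter and score on volume containment, then require minimum 2D distance from Rufus water actor bounds, flat ground, and clearance from indestructible buildings.
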EnvQuery /Lager/Lager/EQS/Lager_EQS_PointInVolume_GenerateToVolumeBounds_OpenSpace_Rufus
options [{
  generator {
   name=EnvQueryGenerator_SimpleGrid_1
   class=EnvQueryGenerator_SimpleGrid
   properties {
    gridSize {
     dataBinding=&self:EnvQueryGenerator_SimpleGrid_1.AIDataProvider_QueryParams_0
     dataField=FloatValue
     defaultValue=20000
    }
    projectionData {
     extentX=0
     navigationFilter=&NavFilter_NPC_ExcludeAllWater_C
     postProjectionVerticalOffset=0
     projectDown=10000
     projectUp=10000
    }
    spaceBetween {
     dataBinding=&self:EnvQueryGenerator_SimpleGrid_1.FortAIDataProvider_FloatCurveOverGameDifficulty_0
     dataField=FloatValue
     defaultValue=2500
    }
   }
  }
  tests [
   EnvQueryTest_Volume_1 {
    class=EnvQueryTest_Volume
    properties {
     boolValue=true
     testPurpose=FilterAndScore
     volumeClass=/Script/LagerRuntime.FortAthenaLivingWorldVolume
     volumeContext=&EnvQueryContext_Querier
    }
   }
   FortQueryTest_DistanceToActorBound_3 {
    class=FortQueryTest_DistanceToActorBound
    properties {
     bDefineReferenceValue=false
     bUse2DDistance=true
     boolValue=true
     clampMaxType=None
     clampMinType=None
     distanceTo=&/EnvQueryContext_RufusWater.EnvQueryContext_RufusWater_C
     filterType=Minimum
     floatValueMax=0
     floatValueMin=500
     multipleContextFilterOp=AllPass
     multipleContextScoreOp=AverageScore
     normalizationType=Absolute
     referenceValue=0
     scoreClampMax=0
     scoreClampMin=0
     scoringEquation=Linear
     scoringFactor=1
     testComment=""
     testPurpose=Filter
    }
   }
   FortQueryTest_OnFlatSurface_1 {class=FortQueryTest_OnFlatSurface, properties {radius=750, toleranceZ=200, traceOffsetDown=200, traceOffsetUp=200}}
   FortQueryTest_DistanceToIndestructibleBuilding_0 {
    class=FortQueryTest_DistanceToIndestructibleBuilding
    properties {
     bDefineReferenceValue=false
     boolValue=true
     clampMaxType=None
     clampMinType=None
     filterType=Minimum
     floatValueMax=0
     floatValueMin=750
     multipleContextFilterOp=AllPass
     multipleContextScoreOp=AverageScore
     normalizationType=Absolute
     referenceValue=0
     scoreClampMax=0
     scoreClampMin=0
     scoringEquation=Linear
     scoringFactor=1
     testComment=""
     testPurpose=Filter
    }
   }
  ]
 }]
