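# EQS query that finds shoreline points along rivers inside a Living World volume for the Lager system.
# A water shoreline generator samples around all river water bodies using a curve data asset and projects points onto navmesh with a filter excluding all water.
# Tests check points against the querier's volume, require a minimum distance from Tower Ruins POI volume bounds, and reject Smackie exclusion tagged areas, treated as infinitely tall.
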EnvQuery /Lager/Lager/EQS/Lager_EQS_PointInVolume_GenerateToVolumeBounds_Shorelines
options [{
  generator {
   name=FortQueryGenerator_PointsOnWaterShoreLine_0
   class=FortQueryGenerator_PointsOnWaterShoreLine
   properties {
    bAutoSortTests=true
    bCanRunAsync=false
    curvesAroundLineAsset=/Game/Athena/AI/EQS/QueryData/EnvQueryData_AlongShorelineCurve
    envQueryResultNormalizationOption=Default
    generateAround=/Game/Athena/AI/EQS/Contexts/EnvQueryContext_AllRiversWaterBodies.EnvQueryContext_AllRiversWaterBodies_C
    navDataOverrideContext=null
    optionName=""
    segmentMaximumVerticalDegreeAngle=90
    projectionData {
     bCanDisableTrace=true
     bCanProjectDown=true
     bCanTraceOnGeometry=true
     bCanTraceOnNavMesh=true
     bOnlyBlockingHits=true
     bTraceComplex=false
     extentX=0
     extentY=10
     extentZ=10
     navigationFilter=&NavFilter_NPC_ExcludeAllWater_C
     postProjectionVerticalOffset=15
     projectDown=1024
     projectUp=1024
     serializedChannel=ECC_Visibility
     traceChannel=TraceTypeQuery1
     traceMode=Navigation
     traceProfileName=None
     traceShape=Line
     versionNum=1
    }
   }
  }
  tests [
   EnvQueryTest_Volume_0 {
    class=EnvQueryTest_Volume
    properties {
     boolValue=true
     testPurpose=FilterAndScore
     volumeClass=/Script/LagerRuntime.FortAthenaLivingWorldVolume
     volumeContext=&EnvQueryContext_Querier
    }
   }
   FortQueryTest_DistanceToActorBound_0 {
    class=FortQueryTest_DistanceToActorBound
    properties {
     bDefineReferenceValue=false
     bUse2DDistance=false
     boolValue=true
     clampMaxType=None
     clampMinType=None
     distanceTo=/Game/Athena/AI/EQS/Contexts/EnvQueryContext_TowerRuinsPOIVolume.EnvQueryContext_TowerRuinsPOIVolume_C
     filterType=Minimum
     floatValueMax=0
     floatValueMin=1
     multipleContextFilterOp=AllPass
     multipleContextScoreOp=AverageScore
     normalizationType=Absolute
     referenceValue=0
     scoreClampMax=0
     scoreClampMin=0
     scoringEquation=Linear
     scoringFactor=1
     testComment=""
     testPurpose=Filter
    }
   }
   FortQueryTest_WithinTaggedArea_0 {
    class=FortQueryTest_WithinTaggedArea
    properties {
     bAssumeInfiniteHeightForArea=true
     bDefineReferenceValue=false
     boolValue=false
     clampMaxType=None
     clampMinType=None
     filterType=Match
     floatValueMax=0
     floatValueMin=0
     multipleContextFilterOp=AllPass
     multipleContextScoreOp=AverageScore
     normalizationType=Absolute
     referenceValue=0
     scoreClampMax=0
     scoreClampMin=0
     scoringEquation=Constant
     scoringFactor=1
     testComment=""
     testPurpose=Filter
     areaExtentBuffer=<0 0 0>
     tagQuery {
      autoDescription=" ANY( Athena.AI.Lager.Exclusion.Smackie )"
      tokenStreamVersion=0
      userDescription=""
      queryTokenStream [0,1,1,1,0]
      tagDictionary [{tagName=Athena.AI.Lager.Exclusion.Smackie}]
     }
    }
   }
  ]
 }]
