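# EQS query that picks open-ground points inside the querier's Living World volume for weather events, apparently lightning storms and tornadoes.
# A simple grid, sized by query parameter and spaced by a difficulty curve, projects onto navmesh excluding water.
# Tests require distance from major POIs, lakes, rivers and buildings, clear upward sweeps, non-destructible flat ground.
# Points inside data-table weather exclusion zones or buffered Tempest exclusion-tagged areas are rejected.
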EnvQuery /Lager/Lager/EQS/Lager_EQS_PointInVolume_GenerateToVolumeBounds_Weather
options [{
  generator {
   name=EnvQueryGenerator_SimpleGrid_1
   class=EnvQueryGenerator_SimpleGrid
   properties {
    gridSize {
     dataBinding=&self:EnvQueryGenerator_SimpleGrid_1.AIDataProvider_QueryParams_0
     dataField=FloatValue
     defaultValue=20000
    }
    projectionData {
     extentX=0
     navigationFilter=&NavFilter_NPC_ExcludeAllWater_C
     postProjectionVerticalOffset=0
     projectDown=3000
     projectUp=3000
    }
    spaceBetween {
     dataBinding=&self:EnvQueryGenerator_SimpleGrid_1.FortAIDataProvider_FloatCurveOverGameDifficulty_0
     dataField=FloatValue
     defaultValue=2500
    }
   }
  }
  tests [
   EnvQueryTest_Volume_0 {
    class=EnvQueryTest_Volume
    properties {
     boolValue=true
     testPurpose=FilterAndScore
     volumeClass=/Script/LagerRuntime.FortAthenaLivingWorldVolume
     volumeContext=&EnvQueryContext_Querier
    }
   }
   FortQueryTest_DistanceToActorBound_0 {
    class=FortQueryTest_DistanceToActorBound
    properties {
     bDefineReferenceValue=false
     bUse2DDistance=false
     boolValue=true
     clampMaxType=None
     clampMinType=None
     distanceTo=/Game/Athena/AI/EQS/Contexts/EnvQueryContext_AllMajorPOI.EnvQueryContext_AllMajorPOI_C
     filterType=Minimum
     floatValueMax=0
     floatValueMin=5000
     multipleContextFilterOp=AllPass
     multipleContextScoreOp=AverageScore
     normalizationType=Absolute
     referenceValue=0
     scoreClampMax=0
     scoreClampMin=0
     scoringEquation=Linear
     scoringFactor=1
     testComment=POI
     testPurpose=Filter
    }
   }
   FortQueryTest_DistanceToActorBound_1 {
    class=FortQueryTest_DistanceToActorBound
    properties {
     bDefineReferenceValue=false
     bUse2DDistance=false
     boolValue=true
     clampMaxType=None
     clampMinType=None
     distanceTo=/Game/Athena/AI/EQS/Contexts/EnvQueryContext_AllLakeAndRiverWaterBodies.EnvQueryContext_AllLakeAndRiverWaterBodies_C
     filterType=Minimum
     floatValueMax=0
     floatValueMin=3000
     multipleContextFilterOp=AllPass
     multipleContextScoreOp=AverageScore
     normalizationType=Absolute
     referenceValue=0
     scoreClampMax=0
     scoreClampMin=0
     scoringEquation=Linear
     scoringFactor=1
     testComment=""
     testPurpose=Filter
    }
   }
   EnvQueryTest_Project_0 {
    class=EnvQueryTest_Project
    properties {
     boolValue=false
     scoringFactor=-1
     testComment=Require clear sweep up
     testPurpose=Filter
     projectionData {
      extentY=0
      extentZ=100
      navigationFilter=&RecastFilter_UseDefaultArea
      postProjectionVerticalOffset=0
      projectDown=0
      projectUp=800
      traceMode=GeometryByChannel
     }
    }
   }
   EnvQueryTest_Project_4 {
    class=EnvQueryTest_Project
    properties {
     scoringFactor=1
     testComment=Require NonDestructible
     testPurpose=Filter
     projectionData {
      extentX=70
      extentZ=135
      navigationFilter=null
      postProjectionVerticalOffset=0
      projectDown=256
      projectUp=0
      serializedChannel=ECC_GameTraceChannel6
      traceChannel=TraceTypeQuery7
      traceMode=GeometryByChannel
      traceShape=Capsule
     }
    }
   }
   FortQueryTest_HasNearbyBuildings_0 {class=FortQueryTest_HasNearbyBuildings, properties {extentXY=5, testComment=Require no building nearby, testPurpose=FilterAndScore}}
   FortQueryTest_OnFlatSurface_0 {class=FortQueryTest_OnFlatSurface, properties {radius=2000, toleranceZ=500, traceOffsetDown=500, traceOffsetUp=500}}
   FortQueryTest_WithinHotfixVolumeBounds_0 {
    class=FortQueryTest_WithinHotfixVolumeBounds
    properties {
     bDefineReferenceValue=false
     bXYOnly=true
     boolValue=false
     boundsTable=/Game/Athena/AI/Weather/DataTables/WeatherExclusionZones
     clampMaxType=None
     clampMinType=None
     filterType=Match
     floatValueMax=0
     floatValueMin=0
     multipleContextFilterOp=AllPass
     multipleContextScoreOp=AverageScore
     normalizationType=Absolute
     referenceValue=0
     scoreClampMax=0
     scoreClampMin=0
     scoringEquation=Constant
     scoringFactor=1
     testComment=Not in exclusion zone
     testPurpose=FilterAndScore
     boundsExtentBuffer=<0 0 0>
    }
   }
   FortQueryTest_WithinTaggedArea_0 {
    class=FortQueryTest_WithinTaggedArea
    properties {
     bAssumeInfiniteHeightForArea=true
     bDefineReferenceValue=false
     boolValue=false
     clampMaxType=None
     clampMinType=None
     filterType=Match
     floatValueMax=0
     floatValueMin=0
     multipleContextFilterOp=AllPass
     multipleContextScoreOp=AverageScore
     normalizationType=Absolute
     referenceValue=0
     scoreClampMax=0
     scoreClampMin=0
     scoringEquation=Constant
     scoringFactor=1
     testComment=""
     testPurpose=Filter
     areaExtentBuffer=<10000 10000 0>
     tagQuery {
      autoDescription=" ANY( Athena.AI.Tempest.Exclusion.All, Athena.AI.Tempest.Exclusion.LightningStorm, Athena.AI.Tempest.Exclusion.Tornado )"
      tokenStreamVersion=0
      userDescription=""
      queryTokenStream [0,1,1,3,0,1,2]
      tagDictionary [
       {tagName=Athena.AI.Tempest.Exclusion.All}
       {tagName=Athena.AI.Tempest.Exclusion.LightningStorm}
       {tagName=Athena.AI.Tempest.Exclusion.Tornado}
      ]
     }
    }
   }
  ]
 }]
