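# EQS query that finds water points inside Living World volumes, using the Lager runtime's FortAthenaLivingWorldVolume class.
# A wide, coarse SimpleGrid generator projects points vertically onto navigation through an NPC water nav filter that excludes other areas.
# A single Volume test then filters and scores those points on being inside that volume class, with the Querier as its volume context.
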EnvQuery /Lager/Lager/EQS/Lager_EQS_PointInVolume_Water
options [{
  generator {
   name=EnvQueryGenerator_SimpleGrid_0
   class=EnvQueryGenerator_SimpleGrid
   properties {
    gridSize=7500
    spaceBetween=1000
    projectionData {
     extentX=0
     navigationFilter=/Game/Athena/AI/NavFilters/NavFilter_NPC_Water_ExcludeOthers.NavFilter_NPC_Water_ExcludeOthers_C
     postProjectionVerticalOffset=0
     projectDown=1024
     projectUp=1024
    }
   }
  }
  tests [EnvQueryTest_Volume_0 {
    class=EnvQueryTest_Volume
    properties {
     boolValue=true
     testPurpose=FilterAndScore
     volumeClass=/Script/LagerRuntime.FortAthenaLivingWorldVolume
     volumeContext=&EnvQueryContext_Querier
    }
   }]
 }]
