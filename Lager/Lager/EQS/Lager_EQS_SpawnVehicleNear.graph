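# Environment query that, by its name, finds candidate spots for spawning a vehicle nearby in Lager.
# A single OnCircle generator places four points on a ring around the default center.
# Its projection settings use a navigation filter that excludes all water for NPCs, with a deep downward search and no horizontal extent.
# The query has no tests, so candidates are neither scored nor filtered beyond generation.
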EnvQuery /Lager/Lager/EQS/Lager_EQS_SpawnVehicleNear
options [{
  generator {
   name=EnvQueryGenerator_OnCircle_0
   class=EnvQueryGenerator_OnCircle
   properties {
    circleRadius=500
    numberOfPoints=4
    projectionData {
     extentX=0
     navigationFilter=&NavFilter_NPC_ExcludeAllWater_C
     postProjectionVerticalOffset=0
     projectDown=1024
    }
    traceData {navigationFilter=null}
   }
  }
  tests []
 }]
